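# Debug widget row that shows one audio motor sim component's class name, cached inputs and runtime values, such as throttle, RPM, gear, volume and pitch.
# UpdateText casts the current and previous sim entries to AudioMotorSimComponent and passes both sets of values to helper functions.
# UpdateFloatText colors numbers green when rising, red when falling and white when unchanged.
# UpdateBoolText shows True/False, green when changed and white otherwise.
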WidgetBlueprint /Game/Sounds/Fort_Audio_Blueprints/Utilities/EditorUtilityWidgets/MotorSim/MotorSimDebugComponentEntry
graphs [
 EventGraph {links [], nodes []}
 UpdateBoolText {nodes [
   n1 = UpdateBoolText!() -> (then, New:bool=~, Old:bool=~, TextBlock:Common Text Object Reference) {is=title K2Node_FunctionEntry}
   > Set Text(Text=Select(Option 0:Text="LOCTEXT(\"False\")", Option 1:Text="LOCTEXT(\"True\")", Index=@n1) -> (ReturnValue:Text), self=@n1.TextBlock)
   > n2 = Branch(Condition="Equal (Boolean)"(A:bool=@n1, B:bool=@n1.Old) -> (ReturnValue:bool=~) {is=title K2Node_PromotableOperator})
   > SetColorAndOpacity!(
    execute
    self=@n1.TextBlock
    InColorAndOpacity_SpecifiedColor:Linear Color Structure=MakeColor!(R:float=1, G:float=1, B:float=1, A:float=1) -> (ReturnValue:Linear Color Structure)
    InColorAndOpacity_ColorUseRule:ESlateColorStylingMode Enum
   ) -> (then)
   SetColorAndOpacity(execute=@n2.else, self=@n1.TextBlock, InColorAndOpacity_SpecifiedColor=MakeColor(R=0, G=1, B=0, A=1))
  ]}
 UpdateFloatText {nodes [
   n1 = UpdateFloatText!() -> (then, NewNum:double=~, OldNum:double=~, TextBlock:Common Numeric Text Block Object Reference) {is=title K2Node_FunctionEntry}
   > SetCurrentValue!(execute, self:Common Numeric Text Block Object Reference=@n1.TextBlock, NewValue:float=@n1) -> (then)
   > n2 = Branch(Condition="float > float"(A=@n1, B=@n1.OldNum))
   > SetColorAndOpacity(self=@n1.TextBlock, InColorAndOpacity_SpecifiedColor=MakeColor(R=0, G=1, B=0, A=1))
   n3 = Branch(execute=@n2.else, Condition="float < float"(A=@n1, B=@n1.OldNum))
   > SetColorAndOpacity(self=@n1.TextBlock, InColorAndOpacity_SpecifiedColor=MakeColor(R=1, G=0, B=0, A=1))
   Branch(execute=@n3.else)
   > SetColorAndOpacity(self=@n1.TextBlock, InColorAndOpacity_SpecifiedColor=MakeColor(R=1, G=1, B=1, A=1))
  ]}
 UpdateText {nodes [
   n1 = UpdateText!() -> (then, SimEntry_Sim:Audio Motor Sim Interface, SimEntry_SortOrder:int, PrevSimEntry_Sim:Audio Motor Sim Interface, PrevSimEntry_SortOrder:int) {is=title K2Node_FunctionEntry}
   > n2 = Cast To AudioMotorSimComponent!(execute, Object="To Object (Interface)"!(Interface:"Interface Interface (by ref)"=@n1) -> (ReturnValue:Object Reference)) -> (then, CastFailed, AsAudio Motor Sim Component:Audio Motor Sim Component Object Reference) {is=title K2Node_DynamicCast}
   > Set NewComponent(NewComponent:Audio Motor Sim Component Object Reference=@n2) -> (Output_Get:Audio Motor Sim Component Object Reference)
   > n3 = Cast To AudioMotorSimComponent(Object:Audio Motor Sim Interface=@n1.PrevSimEntry_Sim)
   > Set PrevComponent(PrevComponent:Audio Motor Sim Component Object Reference=@n3) -> (Output_Get:Audio Motor Sim Component Object Reference)
   n4 = Get Class(Object=@n2) -> (ReturnValue:Audio Motor Sim Component Class Reference)
   > Set Text(
    Text="To Text (String)"(InString=Get Class Display Name(Class=@n4))
    self=Get ComponentName() -> (ComponentName:Common Text Object Reference)
   )
   > n5 = Sequence()
   n6 = Get!(self:Audio Motor Sim Component Object Reference=Get PrevComponent() -> (PrevComponent:Audio Motor Sim Component Object Reference)) -> (
    CachedRuntimeInfo_bShifting:bool=~
    CachedRuntimeInfo_Gear:int
    CachedRuntimeInfo_VirtualSpeedKmh:float=~
    CachedRuntimeInfo_Rpm:float=~
    CachedRuntimeInfo_Volume:float=~
    CachedRuntimeInfo_Pitch:float=~
   )
   n7 = Get(self=Get NewComponent() -> (NewComponent:Audio Motor Sim Component Object Reference)) -> (
    CachedRuntimeInfo_bShifting=~
    CachedRuntimeInfo_VirtualSpeedKmh=~
    CachedRuntimeInfo_Rpm=~
    CachedRuntimeInfo_Volume=~
    CachedRuntimeInfo_Pitch=~
   )
   > UpdateFloatText!(
    execute
    self
    NewNum:double="To Float (Integer)"(InInt=@n7.CachedRuntimeInfo_Gear)
    OldNum:double="To Float (Integer)"(InInt=@n6.CachedRuntimeInfo_Gear)
    TextBlock:Common Numeric Text Block Object Reference=Get GearNum() -> (GearNum:Common Numeric Text Block Object Reference)
   ) -> (then) {is=title K2Node_CallFunction}
   > UpdateFloatText(
    NewNum=@n7.CachedRuntimeInfo_Rpm
    OldNum=@n6.CachedRuntimeInfo_Rpm
    TextBlock=Get RPMNumber() -> (RPMNumber:Common Numeric Text Block Object Reference)
   )
   > UpdateFloatText(
    NewNum=@n7.CachedRuntimeInfo_Volume
    OldNum=@n6.CachedRuntimeInfo_Volume
    TextBlock=Get VolumeNum() -> (VolumeNum:Common Numeric Text Block Object Reference)
   )
   > UpdateFloatText(
    NewNum=@n7.CachedRuntimeInfo_Pitch
    OldNum=@n6.CachedRuntimeInfo_Pitch
    TextBlock=Get PitchNum() -> (PitchNum:Common Numeric Text Block Object Reference)
   )
   n8 = Get!(self=Get PrevComponent()) -> (
    CachedInput_DeltaTime:float=~
    CachedInput_Speed:float=~
    CachedInput_ForwardSpeed:float=~
    CachedInput_SideSpeed:float=~
    CachedInput_UpSpeed:float=~
    CachedInput_Throttle:float=~
    CachedInput_Brake:float=~
    CachedInput_SurfaceFrictionModifier:float=~
    CachedInput_MotorFrictionModifier:float=~
    CachedInput_Boost:float=~
    CachedInput_bDriving:bool=~
    CachedInput_bGrounded:bool=~
    CachedInput_bCanShift:bool=~
    CachedInput_bClutchEngaged:bool=~
   )
   n9 = Get(self=Get NewComponent()) -> (
    CachedInput_DeltaTime=~
    CachedInput_Speed=~
    CachedInput_ForwardSpeed=~
    CachedInput_SideSpeed=~
    CachedInput_UpSpeed=~
    CachedInput_Throttle=~
    CachedInput_Brake=~
    CachedInput_SurfaceFrictionModifier=~
    CachedInput_MotorFrictionModifier=~
    CachedInput_Boost=~
    CachedInput_bDriving=~
    CachedInput_bGrounded=~
    CachedInput_bCanShift=~
    CachedInput_bClutchEngaged=~
   )
   UpdateFloatText(
    execute=@n5.then_1
    NewNum=@n9.CachedInput_Throttle
    OldNum=@n8.CachedInput_Throttle
    TextBlock=Get ThrottleNum() -> (ThrottleNum:Common Numeric Text Block Object Reference)
   )
   > UpdateFloatText(
    NewNum=@n9.CachedInput_Speed
    OldNum=@n8.CachedInput_Speed
    TextBlock=Get SpeedNum() -> (SpeedNum:Common Numeric Text Block Object Reference)
   )
   > UpdateFloatText(
    NewNum=@n9.CachedInput_ForwardSpeed
    OldNum=@n8.CachedInput_ForwardSpeed
    TextBlock=Get FwdSpeedNum() -> (FwdSpeedNum:Common Numeric Text Block Object Reference)
   )
   > UpdateFloatText(
    NewNum=@n9.CachedInput_SideSpeed
    OldNum=@n8.CachedInput_SideSpeed
    TextBlock=Get SideSpeedNum() -> (SideSpeedNum:Common Numeric Text Block Object Reference)
   )
   > UpdateFloatText(
    NewNum=@n9.CachedInput_UpSpeed
    OldNum=@n8.CachedInput_UpSpeed
    TextBlock=Get UpSpeedNum() -> (UpSpeedNum:Common Numeric Text Block Object Reference)
   )
   > UpdateFloatText(
    NewNum=@n9.CachedInput_Brake
    OldNum=@n8.CachedInput_Brake
    TextBlock=Get BrakeNum() -> (BrakeNum:Common Numeric Text Block Object Reference)
   )
   > UpdateFloatText(
    NewNum=@n9.CachedInput_SurfaceFrictionModifier
    OldNum=@n8.CachedInput_SurfaceFrictionModifier
    TextBlock=Get SurfaceFrictionNum() -> (SurfaceFrictionNum:Common Numeric Text Block Object Reference)
   )
   > UpdateFloatText(
    NewNum=@n9.CachedInput_MotorFrictionModifier
    OldNum=@n8.CachedInput_MotorFrictionModifier
    TextBlock=Get MotorFrictionNum() -> (MotorFrictionNum:Common Numeric Text Block Object Reference)
   )
   > UpdateFloatText(
    NewNum=@n9.CachedInput_Boost
    OldNum=@n8.CachedInput_Boost
    TextBlock=Get BoostNum() -> (BoostNum:Common Numeric Text Block Object Reference)
   )
   > UpdateBoolText!(
    execute
    self
    New:bool=@n9.CachedInput_bClutchEngaged
    Old:bool=@n8.CachedInput_bClutchEngaged
    TextBlock:Common Text Object Reference=Get ClutchValue() -> (ClutchValue:Common Text Object Reference)
   ) -> (then) {is=title K2Node_CallFunction}
   > UpdateBoolText(New=@n9.CachedInput_bDriving, Old=@n8.CachedInput_bDriving, TextBlock=Get DriveValue() -> (DriveValue:Common Text Object Reference))
   > UpdateBoolText(New=@n9.CachedInput_bCanShift, Old=@n8.CachedInput_bCanShift, TextBlock=Get ShiftValue() -> (ShiftValue:Common Text Object Reference))
  ]}
]
variables []
widgets {
 info {inheritedWidgetCount=0, parentClass=&UserWidget, rootWidgetClass=&HorizontalBox, widgetCount=52}
 widgets [
  HorizontalBox_22 {widgetClassPath=&HorizontalBox, widget=HorizontalBox_22}
  HorizontalBox_136 {
   widgetClassPath=&HorizontalBox
   parent=HorizontalBox_22
   widget=HorizontalBox_136
   slot {class=HorizontalBoxSlot, properties {horizontalAlignment=HAlign_Right, padding=<ltrb 0 0 10 0>, size {sizeRule=Fill, value=0.18}}}
  }
  ComponentName {
   widgetClassPath=&CommonTextBlock
   bIsVariable=true
   parent=HorizontalBox_136
   widget=ComponentName
   properties {
    style=&TextStyle-Base_C
    text=ThrottleStateMotorSimComponent
    font {
     bForceMonospaced=false
     bMaterialIsStencil=false
     fontMaterial=null
     fontObject=&NotoSans
     letterSpacing=0
     monospacedWidth=1
     size=24
     skewAmount=0
     typefaceFontName=NotoSans-Regular
     outlineSettings {
      bApplyOutlineToDropShadows=false
      bMiteredCorners=false
      bSeparateFillAlpha=false
      outlineMaterial=null
      outlineSize=0
      outlineColor=<rgba 0 0 0 1>
     }
    }
    renderTransform {angle=0, scale=<0.6 0.6>, shear=<-2 0>, translation=<-95.2526 -1.9964>}
    shadowOffset=<0 0>
   }
   slot {
    class=HorizontalBoxSlot
    properties {horizontalAlignment=HAlign_Right, verticalAlignment=VAlign_Top, padding=<ltrb 1 5 20 10>, size {sizeRule=Automatic, value=0.18}}
   }
  }
  HorizontalBox_93 {
   widgetClassPath=&HorizontalBox
   parent=HorizontalBox_22
   widget=HorizontalBox_93
   slot {class=HorizontalBoxSlot, properties {padding=<ltrb 10 0 0 0>, size {sizeRule=Fill, value=1}}}
  }
  VerticalBox_100 {widgetClassPath=&VerticalBox, parent=HorizontalBox_93, widget=VerticalBox_100, slot {class=HorizontalBoxSlot}}
  ThrottleName {
   widgetClassPath=&CommonTextBlock
   bIsVariable=true
   parent=VerticalBox_100
   widget=ThrottleName
   properties {
    style=&TextStyle-Base_C
    text="Throttle "
    font {
     bForceMonospaced=false
     bMaterialIsStencil=false
     fontMaterial=null
     fontObject=&NotoSans
     letterSpacing=0
     monospacedWidth=1
     size=24
     skewAmount=0
     typefaceFontName=NotoSans-Regular
     outlineSettings {
      bApplyOutlineToDropShadows=false
      bMiteredCorners=false
      bSeparateFillAlpha=false
      outlineMaterial=null
      outlineSize=0
      outlineColor=<rgba 0 0 0 1>
     }
    }
    renderTransform {angle=0, scale=<0.7 0.7>, shear=<0 0>, translation=<0 0>}
    shadowOffset=<0 0>
   }
   slot {
    class=VerticalBoxSlot
    properties {horizontalAlignment=HAlign_Left, verticalAlignment=VAlign_Top, padding=<ltrb 10 5 10 0>, size {sizeRule=Fill, value=1}}
   }
  }
  ThrottleNum {
   widgetClassPath=/Script/CommonUI.CommonNumericTextBlock
   bIsVariable=true
   parent=VerticalBox_100
   widget=ThrottleNum
   properties {
    bIsScrollingEnabled=false
    bIsVolatile=true
    style=&TextStyle-Base_C
    text=Text Block
    font {
     bForceMonospaced=false
     bMaterialIsStencil=false
     fontMaterial=null
     fontObject=&NotoSans
     letterSpacing=0
     monospacedWidth=1
     size=24
     skewAmount=0
     typefaceFontName=NotoSans-Regular
     outlineSettings {
      bApplyOutlineToDropShadows=false
      bMiteredCorners=false
      bSeparateFillAlpha=false
      outlineMaterial=null
      outlineSize=0
      outlineColor=<rgba 0 0 0 1>
     }
    }
    formattingSpecification {
     alwaysSign=false
     maximumFractionalDigits=2
     maximumIntegralDigits=2
     minimumFractionalDigits=2
     minimumIntegralDigits=1
     roundingMode=HalfFromZero
     useGrouping=true
    }
    renderTransform {angle=0, scale=<0.6 0.6>, shear=<0 0>, translation=<0 0>}
    shadowOffset=<0 0>
   }
   slot {
    class=VerticalBoxSlot
    properties {horizontalAlignment=HAlign_Center, verticalAlignment=VAlign_Top, padding=<ltrb 1 1 1 1>, size {sizeRule=Fill, value=1}}
   }
  }
  VerticalBox_8 {widgetClassPath=&VerticalBox, parent=HorizontalBox_93, widget=VerticalBox_8, slot {class=HorizontalBoxSlot}}
  BrakeName {
   widgetClassPath=&CommonTextBlock
   bIsVariable=true
   parent=VerticalBox_8
   widget=BrakeName
   properties {
    style=&TextStyle-Base_C
    text=Brake
    font {
     bForceMonospaced=false
     bMaterialIsStencil=false
     fontMaterial=null
     fontObject=&NotoSans
     letterSpacing=0
     monospacedWidth=1
     size=24
     skewAmount=0
     typefaceFontName=NotoSans-Regular
     outlineSettings {
      bApplyOutlineToDropShadows=false
      bMiteredCorners=false
      bSeparateFillAlpha=false
      outlineMaterial=null
      outlineSize=0
      outlineColor=<rgba 0 0 0 1>
     }
    }
    renderTransform {angle=0, scale=<0.7 0.7>, shear=<0 0>, translation=<0 0>}
    shadowOffset=<0 0>
   }
   slot {
    class=VerticalBoxSlot
    properties {horizontalAlignment=HAlign_Left, verticalAlignment=VAlign_Top, padding=<ltrb 10 5 10 0>, size {sizeRule=Fill, value=1}}
   }
  }
  BrakeNum {
   widgetClassPath=&CommonNumericTextBlock
   bIsVariable=true
   parent=VerticalBox_8
   widget=BrakeNum
   properties {
    bIsScrollingEnabled=false
    bIsVolatile=true
    style=&TextStyle-Base_C
    text=Text Block
    font {
     bForceMonospaced=false
     bMaterialIsStencil=false
     fontMaterial=null
     fontObject=&NotoSans
     letterSpacing=0
     monospacedWidth=1
     size=24
     skewAmount=0
     typefaceFontName=NotoSans-Regular
     outlineSettings {
      bApplyOutlineToDropShadows=false
      bMiteredCorners=false
      bSeparateFillAlpha=false
      outlineMaterial=null
      outlineSize=0
      outlineColor=<rgba 0 0 0 1>
     }
    }
    formattingSpecification {
     alwaysSign=false
     maximumFractionalDigits=0
     maximumIntegralDigits=2
     minimumFractionalDigits=2
     minimumIntegralDigits=1
     roundingMode=HalfFromZero
     useGrouping=true
    }
    renderTransform {angle=0, scale=<0.6 0.6>, shear=<0 0>, translation=<0 0>}
    shadowOffset=<0 0>
   }
   slot {
    class=VerticalBoxSlot
    properties {horizontalAlignment=HAlign_Center, verticalAlignment=VAlign_Top, padding=<ltrb 1 1 1 1>, size {sizeRule=Fill, value=1}}
   }
  }
  VerticalBox_46 {widgetClassPath=&VerticalBox, parent=HorizontalBox_93, widget=VerticalBox_46, slot {class=HorizontalBoxSlot}}
  RPMName {
   widgetClassPath=&CommonTextBlock
   bIsVariable=true
   parent=VerticalBox_46
   widget=RPMName
   properties {
    style=&TextStyle-Base_C
    text=RPM
    font {
     bForceMonospaced=false
     bMaterialIsStencil=false
     fontMaterial=null
     fontObject=&NotoSans
     letterSpacing=0
     monospacedWidth=1
     size=24
     skewAmount=0
     typefaceFontName=NotoSans-Regular
     outlineSettings {
      bApplyOutlineToDropShadows=false
      bMiteredCorners=false
      bSeparateFillAlpha=false
      outlineMaterial=null
      outlineSize=0
      outlineColor=<rgba 0 0 0 1>
     }
    }
    renderTransform {angle=0, scale=<0.7 0.7>, shear=<0 0>, translation=<0 0>}
    shadowOffset=<0 0>
   }
   slot {
    class=VerticalBoxSlot
    properties {horizontalAlignment=HAlign_Left, verticalAlignment=VAlign_Top, padding=<ltrb 10 5 10 0>, size {sizeRule=Fill, value=1}}
   }
  }
  RPMNumber {
   widgetClassPath=&CommonNumericTextBlock
   bIsVariable=true
   parent=VerticalBox_46
   widget=RPMNumber
   properties {
    bIsScrollingEnabled=false
    bIsVolatile=true
    style=&TextStyle-Base_C
    text=Text Block
    font {
     bForceMonospaced=false
     bMaterialIsStencil=false
     fontMaterial=null
     fontObject=&NotoSans
     letterSpacing=0
     monospacedWidth=1
     size=24
     skewAmount=0
     typefaceFontName=NotoSans-Regular
     outlineSettings {
      bApplyOutlineToDropShadows=false
      bMiteredCorners=false
      bSeparateFillAlpha=false
      outlineMaterial=null
      outlineSize=0
      outlineColor=<rgba 0 0 0 1>
     }
    }
    formattingSpecification {
     alwaysSign=false
     maximumFractionalDigits=2
     maximumIntegralDigits=2
     minimumFractionalDigits=2
     minimumIntegralDigits=1
     roundingMode=HalfFromZero
     useGrouping=true
    }
    renderTransform {angle=0, scale=<0.6 0.6>, shear=<0 0>, translation=<0 0>}
    shadowOffset=<0 0>
   }
   slot {
    class=VerticalBoxSlot
    properties {horizontalAlignment=HAlign_Center, verticalAlignment=VAlign_Top, padding=<ltrb 1 1 1 1>, size {sizeRule=Fill, value=1}}
   }
  }
  VerticalBox_1 {widgetClassPath=&VerticalBox, parent=HorizontalBox_93, widget=VerticalBox_1, slot {class=HorizontalBoxSlot}}
  GearName {
   widgetClassPath=&CommonTextBlock
   bIsVariable=true
   parent=VerticalBox_1
   widget=GearName
   properties {
    style=&TextStyle-Base_C
    text=Gear
    font {
     bForceMonospaced=false
     bMaterialIsStencil=false
     fontMaterial=null
     fontObject=&NotoSans
     letterSpacing=0
     monospacedWidth=1
     size=24
     skewAmount=0
     typefaceFontName=NotoSans-Regular
     outlineSettings {
      bApplyOutlineToDropShadows=false
      bMiteredCorners=false
      bSeparateFillAlpha=false
      outlineMaterial=null
      outlineSize=0
      outlineColor=<rgba 0 0 0 1>
     }
    }
    renderTransform {angle=0, scale=<0.7 0.7>, shear=<0 0>, translation=<0 0>}
    shadowOffset=<0 0>
   }
   slot {
    class=VerticalBoxSlot
    properties {horizontalAlignment=HAlign_Left, verticalAlignment=VAlign_Top, padding=<ltrb 10 5 10 0>, size {sizeRule=Fill, value=1}}
   }
  }
  GearNum {
   widgetClassPath=&CommonNumericTextBlock
   bIsVariable=true
   parent=VerticalBox_1
   widget=GearNum
   properties {
    bIsScrollingEnabled=false
    bIsVolatile=true
    style=&TextStyle-Base_C
    text=Text Block
    font {
     bForceMonospaced=false
     bMaterialIsStencil=false
     fontMaterial=null
     fontObject=&NotoSans
     letterSpacing=0
     monospacedWidth=1
     size=24
     skewAmount=0
     typefaceFontName=NotoSans-Regular
     outlineSettings {
      bApplyOutlineToDropShadows=false
      bMiteredCorners=false
      bSeparateFillAlpha=false
      outlineMaterial=null
      outlineSize=0
      outlineColor=<rgba 0 0 0 1>
     }
    }
    formattingSpecification {
     alwaysSign=false
     maximumFractionalDigits=0
     maximumIntegralDigits=2
     minimumFractionalDigits=2
     minimumIntegralDigits=1
     roundingMode=HalfFromZero
     useGrouping=true
    }
    renderTransform {angle=0, scale=<0.6 0.6>, shear=<0 0>, translation=<0 0>}
    shadowOffset=<0 0>
   }
   slot {
    class=VerticalBoxSlot
    properties {horizontalAlignment=HAlign_Center, verticalAlignment=VAlign_Top, padding=<ltrb 1 1 1 1>, size {sizeRule=Fill, value=1}}
   }
  }
  VerticalBox {widgetClassPath=&VerticalBox, parent=HorizontalBox_93, widget=VerticalBox, slot {class=HorizontalBoxSlot}}
  VolumeName {
   widgetClassPath=&CommonTextBlock
   bIsVariable=true
   parent=VerticalBox
   widget=VolumeName
   properties {
    style=&TextStyle-Base_C
    text=Volume
    font {
     bForceMonospaced=false
     bMaterialIsStencil=false
     fontMaterial=null
     fontObject=&NotoSans
     letterSpacing=0
     monospacedWidth=1
     size=24
     skewAmount=0
     typefaceFontName=NotoSans-Regular
     outlineSettings {
      bApplyOutlineToDropShadows=false
      bMiteredCorners=false
      bSeparateFillAlpha=false
      outlineMaterial=null
      outlineSize=0
      outlineColor=<rgba 0 0 0 1>
     }
    }
    renderTransform {angle=0, scale=<0.7 0.7>, shear=<0 0>, translation=<0 0>}
    shadowOffset=<0 0>
   }
   slot {
    class=VerticalBoxSlot
    properties {horizontalAlignment=HAlign_Left, verticalAlignment=VAlign_Top, padding=<ltrb 10 5 10 0>, size {sizeRule=Fill, value=1}}
   }
  }
  VolumeNum {
   widgetClassPath=&CommonNumericTextBlock
   bIsVariable=true
   parent=VerticalBox
   widget=VolumeNum
   properties {
    bIsScrollingEnabled=false
    bIsVolatile=true
    style=&TextStyle-Base_C
    text=Text Block
    font {
     bForceMonospaced=false
     bMaterialIsStencil=false
     fontMaterial=null
     fontObject=&NotoSans
     letterSpacing=0
     monospacedWidth=1
     size=24
     skewAmount=0
     typefaceFontName=NotoSans-Regular
     outlineSettings {
      bApplyOutlineToDropShadows=false
      bMiteredCorners=false
      bSeparateFillAlpha=false
      outlineMaterial=null
      outlineSize=0
      outlineColor=<rgba 0 0 0 1>
     }
    }
    formattingSpecification {
     alwaysSign=false
     maximumFractionalDigits=0
     maximumIntegralDigits=2
     minimumFractionalDigits=2
     minimumIntegralDigits=1
     roundingMode=HalfFromZero
     useGrouping=true
    }
    renderTransform {angle=0, scale=<0.6 0.6>, shear=<0 0>, translation=<0 0>}
    shadowOffset=<0 0>
   }
   slot {
    class=VerticalBoxSlot
    properties {horizontalAlignment=HAlign_Center, verticalAlignment=VAlign_Top, padding=<ltrb 1 1 1 1>, size {sizeRule=Fill, value=1}}
   }
  }
  VerticalBox_3 {widgetClassPath=&VerticalBox, parent=HorizontalBox_93, widget=VerticalBox_3, slot {class=HorizontalBoxSlot}}
  PitchName {
   widgetClassPath=&CommonTextBlock
   bIsVariable=true
   parent=VerticalBox_3
   widget=PitchName
   properties {
    style=&TextStyle-Base_C
    text=Pitch
    font {
     bForceMonospaced=false
     bMaterialIsStencil=false
     fontMaterial=null
     fontObject=&NotoSans
     letterSpacing=0
     monospacedWidth=1
     size=24
     skewAmount=0
     typefaceFontName=NotoSans-Regular
     outlineSettings {
      bApplyOutlineToDropShadows=false
      bMiteredCorners=false
      bSeparateFillAlpha=false
      outlineMaterial=null
      outlineSize=0
      outlineColor=<rgba 0 0 0 1>
     }
    }
    renderTransform {angle=0, scale=<0.7 0.7>, shear=<0 0>, translation=<0 0>}
    shadowOffset=<0 0>
   }
   slot {
    class=VerticalBoxSlot
    properties {horizontalAlignment=HAlign_Left, verticalAlignment=VAlign_Top, padding=<ltrb 10 5 10 0>, size {sizeRule=Fill, value=1}}
   }
  }
  PitchNum {
   widgetClassPath=&CommonNumericTextBlock
   bIsVariable=true
   parent=VerticalBox_3
   widget=PitchNum
   properties {
    bIsScrollingEnabled=false
    bIsVolatile=true
    style=&TextStyle-Base_C
    text=Text Block
    font {
     bForceMonospaced=false
     bMaterialIsStencil=false
     fontMaterial=null
     fontObject=&NotoSans
     letterSpacing=0
     monospacedWidth=1
     size=24
     skewAmount=0
     typefaceFontName=NotoSans-Regular
     outlineSettings {
      bApplyOutlineToDropShadows=false
      bMiteredCorners=false
      bSeparateFillAlpha=false
      outlineMaterial=null
      outlineSize=0
      outlineColor=<rgba 0 0 0 1>
     }
    }
    formattingSpecification {
     alwaysSign=false
     maximumFractionalDigits=0
     maximumIntegralDigits=2
     minimumFractionalDigits=2
     minimumIntegralDigits=1
     roundingMode=HalfFromZero
     useGrouping=true
    }
    renderTransform {angle=0, scale=<0.6 0.6>, shear=<0 0>, translation=<0 0>}
    shadowOffset=<0 0>
   }
   slot {
    class=VerticalBoxSlot
    properties {horizontalAlignment=HAlign_Center, verticalAlignment=VAlign_Top, padding=<ltrb 1 1 1 1>, size {sizeRule=Fill, value=1}}
   }
  }
  VerticalBox_2 {widgetClassPath=&VerticalBox, parent=HorizontalBox_93, widget=VerticalBox_2, slot {class=HorizontalBoxSlot}}
  SpeedName {
   widgetClassPath=&CommonTextBlock
   bIsVariable=true
   parent=VerticalBox_2
   widget=SpeedName
   properties {
    style=&TextStyle-Base_C
    text=Speed
    font {
     bForceMonospaced=false
     bMaterialIsStencil=false
     fontMaterial=null
     fontObject=&NotoSans
     letterSpacing=0
     monospacedWidth=1
     size=24
     skewAmount=0
     typefaceFontName=NotoSans-Regular
     outlineSettings {
      bApplyOutlineToDropShadows=false
      bMiteredCorners=false
      bSeparateFillAlpha=false
      outlineMaterial=null
      outlineSize=0
      outlineColor=<rgba 0 0 0 1>
     }
    }
    renderTransform {angle=0, scale=<0.7 0.7>, shear=<0 0>, translation=<0 0>}
    shadowOffset=<0 0>
   }
   slot {
    class=VerticalBoxSlot
    properties {horizontalAlignment=HAlign_Left, verticalAlignment=VAlign_Top, padding=<ltrb 10 5 10 0>, size {sizeRule=Fill, value=1}}
   }
  }
  SpeedNum {
   widgetClassPath=&CommonNumericTextBlock
   bIsVariable=true
   parent=VerticalBox_2
   widget=SpeedNum
   properties {
    bIsScrollingEnabled=false
    bIsVolatile=true
    style=&TextStyle-Base_C
    text=Text Block
    font {
     bForceMonospaced=false
     bMaterialIsStencil=false
     fontMaterial=null
     fontObject=&NotoSans
     letterSpacing=0
     monospacedWidth=1
     size=24
     skewAmount=0
     typefaceFontName=NotoSans-Regular
     outlineSettings {
      bApplyOutlineToDropShadows=false
      bMiteredCorners=false
      bSeparateFillAlpha=false
      outlineMaterial=null
      outlineSize=0
      outlineColor=<rgba 0 0 0 1>
     }
    }
    formattingSpecification {
     alwaysSign=false
     maximumFractionalDigits=0
     maximumIntegralDigits=2
     minimumFractionalDigits=2
     minimumIntegralDigits=1
     roundingMode=HalfFromZero
     useGrouping=true
    }
    renderTransform {angle=0, scale=<0.6 0.6>, shear=<0 0>, translation=<0 0>}
    shadowOffset=<0 0>
   }
   slot {
    class=VerticalBoxSlot
    properties {horizontalAlignment=HAlign_Center, verticalAlignment=VAlign_Top, padding=<ltrb 1 1 1 1>, size {sizeRule=Fill, value=1}}
   }
  }
  VerticalBox_147 {widgetClassPath=&VerticalBox, parent=HorizontalBox_93, widget=VerticalBox_147, slot {class=HorizontalBoxSlot}}
  FwdSpeedName {
   widgetClassPath=&CommonTextBlock
   bIsVariable=true
   parent=VerticalBox_147
   widget=FwdSpeedName
   properties {
    style=&TextStyle-Base_C
    text=FwdSpeed
    font {
     bForceMonospaced=false
     bMaterialIsStencil=false
     fontMaterial=null
     fontObject=&NotoSans
     letterSpacing=0
     monospacedWidth=1
     size=24
     skewAmount=0
     typefaceFontName=NotoSans-Regular
     outlineSettings {
      bApplyOutlineToDropShadows=false
      bMiteredCorners=false
      bSeparateFillAlpha=false
      outlineMaterial=null
      outlineSize=0
      outlineColor=<rgba 0 0 0 1>
     }
    }
    renderTransform {angle=0, scale=<0.7 0.7>, shear=<0 0>, translation=<0 0>}
    shadowOffset=<0 0>
   }
   slot {
    class=VerticalBoxSlot
    properties {horizontalAlignment=HAlign_Left, verticalAlignment=VAlign_Top, padding=<ltrb 10 5 10 0>, size {sizeRule=Fill, value=1}}
   }
  }
  FwdSpeedNum {
   widgetClassPath=&CommonNumericTextBlock
   bIsVariable=true
   parent=VerticalBox_147
   widget=FwdSpeedNum
   properties {
    bIsScrollingEnabled=false
    bIsVolatile=true
    style=&TextStyle-Base_C
    text=Text Block
    font {
     bForceMonospaced=false
     bMaterialIsStencil=false
     fontMaterial=null
     fontObject=&NotoSans
     letterSpacing=0
     monospacedWidth=1
     size=24
     skewAmount=0
     typefaceFontName=NotoSans-Regular
     outlineSettings {
      bApplyOutlineToDropShadows=false
      bMiteredCorners=false
      bSeparateFillAlpha=false
      outlineMaterial=null
      outlineSize=0
      outlineColor=<rgba 0 0 0 1>
     }
    }
    formattingSpecification {
     alwaysSign=false
     maximumFractionalDigits=0
     maximumIntegralDigits=2
     minimumFractionalDigits=2
     minimumIntegralDigits=1
     roundingMode=HalfFromZero
     useGrouping=true
    }
    renderTransform {angle=0, scale=<0.6 0.6>, shear=<0 0>, translation=<0 0>}
    shadowOffset=<0 0>
   }
   slot {
    class=VerticalBoxSlot
    properties {horizontalAlignment=HAlign_Center, verticalAlignment=VAlign_Top, padding=<ltrb 1 1 1 1>, size {sizeRule=Fill, value=1}}
   }
  }
  VerticalBox_4 {widgetClassPath=&VerticalBox, parent=HorizontalBox_93, widget=VerticalBox_4, slot {class=HorizontalBoxSlot}}
  SideSpeedName {
   widgetClassPath=&CommonTextBlock
   bIsVariable=true
   parent=VerticalBox_4
   widget=SideSpeedName
   properties {
    style=&TextStyle-Base_C
    text=SideSpeed
    font {
     bForceMonospaced=false
     bMaterialIsStencil=false
     fontMaterial=null
     fontObject=&NotoSans
     letterSpacing=0
     monospacedWidth=1
     size=24
     skewAmount=0
     typefaceFontName=NotoSans-Regular
     outlineSettings {
      bApplyOutlineToDropShadows=false
      bMiteredCorners=false
      bSeparateFillAlpha=false
      outlineMaterial=null
      outlineSize=0
      outlineColor=<rgba 0 0 0 1>
     }
    }
    renderTransform {angle=0, scale=<0.7 0.7>, shear=<0 0>, translation=<0 0>}
    shadowOffset=<0 0>
   }
   slot {
    class=VerticalBoxSlot
    properties {horizontalAlignment=HAlign_Left, verticalAlignment=VAlign_Top, padding=<ltrb 10 5 10 0>, size {sizeRule=Fill, value=1}}
   }
  }
  SideSpeedNum {
   widgetClassPath=&CommonNumericTextBlock
   bIsVariable=true
   parent=VerticalBox_4
   widget=SideSpeedNum
   properties {
    bIsScrollingEnabled=false
    bIsVolatile=true
    style=&TextStyle-Base_C
    text=Text Block
    font {
     bForceMonospaced=false
     bMaterialIsStencil=false
     fontMaterial=null
     fontObject=&NotoSans
     letterSpacing=0
     monospacedWidth=1
     size=24
     skewAmount=0
     typefaceFontName=NotoSans-Regular
     outlineSettings {
      bApplyOutlineToDropShadows=false
      bMiteredCorners=false
      bSeparateFillAlpha=false
      outlineMaterial=null
      outlineSize=0
      outlineColor=<rgba 0 0 0 1>
     }
    }
    formattingSpecification {
     alwaysSign=false
     maximumFractionalDigits=0
     maximumIntegralDigits=2
     minimumFractionalDigits=2
     minimumIntegralDigits=1
     roundingMode=HalfFromZero
     useGrouping=true
    }
    renderTransform {angle=0, scale=<0.6 0.6>, shear=<0 0>, translation=<0 0>}
    shadowOffset=<0 0>
   }
   slot {
    class=VerticalBoxSlot
    properties {horizontalAlignment=HAlign_Center, verticalAlignment=VAlign_Top, padding=<ltrb 1 1 1 1>, size {sizeRule=Fill, value=1}}
   }
  }
  VerticalBox_5 {widgetClassPath=&VerticalBox, parent=HorizontalBox_93, widget=VerticalBox_5, slot {class=HorizontalBoxSlot}}
  UpSpeedName {
   widgetClassPath=&CommonTextBlock
   bIsVariable=true
   parent=VerticalBox_5
   widget=UpSpeedName
   properties {
    style=&TextStyle-Base_C
    text=UpSpeed
    font {
     bForceMonospaced=false
     bMaterialIsStencil=false
     fontMaterial=null
     fontObject=&NotoSans
     letterSpacing=0
     monospacedWidth=1
     size=24
     skewAmount=0
     typefaceFontName=NotoSans-Regular
     outlineSettings {
      bApplyOutlineToDropShadows=false
      bMiteredCorners=false
      bSeparateFillAlpha=false
      outlineMaterial=null
      outlineSize=0
      outlineColor=<rgba 0 0 0 1>
     }
    }
    renderTransform {angle=0, scale=<0.7 0.7>, shear=<0 0>, translation=<0 0>}
    shadowOffset=<0 0>
   }
   slot {
    class=VerticalBoxSlot
    properties {horizontalAlignment=HAlign_Left, verticalAlignment=VAlign_Top, padding=<ltrb 10 5 10 0>, size {sizeRule=Fill, value=1}}
   }
  }
  UpSpeedNum {
   widgetClassPath=&CommonNumericTextBlock
   bIsVariable=true
   parent=VerticalBox_5
   widget=UpSpeedNum
   properties {
    bIsScrollingEnabled=false
    bIsVolatile=true
    style=&TextStyle-Base_C
    text=Text Block
    font {
     bForceMonospaced=false
     bMaterialIsStencil=false
     fontMaterial=null
     fontObject=&NotoSans
     letterSpacing=0
     monospacedWidth=1
     size=24
     skewAmount=0
     typefaceFontName=NotoSans-Regular
     outlineSettings {
      bApplyOutlineToDropShadows=false
      bMiteredCorners=false
      bSeparateFillAlpha=false
      outlineMaterial=null
      outlineSize=0
      outlineColor=<rgba 0 0 0 1>
     }
    }
    formattingSpecification {
     alwaysSign=false
     maximumFractionalDigits=0
     maximumIntegralDigits=2
     minimumFractionalDigits=2
     minimumIntegralDigits=1
     roundingMode=HalfFromZero
     useGrouping=true
    }
    renderTransform {angle=0, scale=<0.6 0.6>, shear=<0 0>, translation=<0 0>}
    shadowOffset=<0 0>
   }
   slot {
    class=VerticalBoxSlot
    properties {horizontalAlignment=HAlign_Center, verticalAlignment=VAlign_Top, padding=<ltrb 1 1 1 1>, size {sizeRule=Fill, value=1}}
   }
  }
  VerticalBox_6 {widgetClassPath=&VerticalBox, parent=HorizontalBox_93, widget=VerticalBox_6, slot {class=HorizontalBoxSlot}}
  SurfaceFrictionName {
   widgetClassPath=&CommonTextBlock
   bIsVariable=true
   parent=VerticalBox_6
   widget=SurfaceFrictionName
   properties {
    style=&TextStyle-Base_C
    text=SurfaceFriction
    font {
     bForceMonospaced=false
     bMaterialIsStencil=false
     fontMaterial=null
     fontObject=&NotoSans
     letterSpacing=0
     monospacedWidth=1
     size=24
     skewAmount=0
     typefaceFontName=NotoSans-Regular
     outlineSettings {
      bApplyOutlineToDropShadows=false
      bMiteredCorners=false
      bSeparateFillAlpha=false
      outlineMaterial=null
      outlineSize=0
      outlineColor=<rgba 0 0 0 1>
     }
    }
    renderTransform {angle=0, scale=<0.7 0.7>, shear=<0 0>, translation=<0 0>}
    shadowOffset=<0 0>
   }
   slot {
    class=VerticalBoxSlot
    properties {horizontalAlignment=HAlign_Left, verticalAlignment=VAlign_Top, padding=<ltrb 10 5 10 0>, size {sizeRule=Fill, value=1}}
   }
  }
  SurfaceFrictionNum {
   widgetClassPath=&CommonNumericTextBlock
   bIsVariable=true
   parent=VerticalBox_6
   widget=SurfaceFrictionNum
   properties {
    bIsScrollingEnabled=false
    bIsVolatile=true
    style=&TextStyle-Base_C
    text=Text Block
    font {
     bForceMonospaced=false
     bMaterialIsStencil=false
     fontMaterial=null
     fontObject=&NotoSans
     letterSpacing=0
     monospacedWidth=1
     size=24
     skewAmount=0
     typefaceFontName=NotoSans-Regular
     outlineSettings {
      bApplyOutlineToDropShadows=false
      bMiteredCorners=false
      bSeparateFillAlpha=false
      outlineMaterial=null
      outlineSize=0
      outlineColor=<rgba 0 0 0 1>
     }
    }
    formattingSpecification {
     alwaysSign=false
     maximumFractionalDigits=0
     maximumIntegralDigits=2
     minimumFractionalDigits=2
     minimumIntegralDigits=1
     roundingMode=HalfFromZero
     useGrouping=true
    }
    renderTransform {angle=0, scale=<0.6 0.6>, shear=<0 0>, translation=<0 0>}
    shadowOffset=<0 0>
   }
   slot {
    class=VerticalBoxSlot
    properties {horizontalAlignment=HAlign_Center, verticalAlignment=VAlign_Top, padding=<ltrb 1 1 1 1>, size {sizeRule=Fill, value=1}}
   }
  }
  VerticalBox_7 {widgetClassPath=&VerticalBox, parent=HorizontalBox_93, widget=VerticalBox_7, slot {class=HorizontalBoxSlot}}
  MotorFrictionName {
   widgetClassPath=&CommonTextBlock
   bIsVariable=true
   parent=VerticalBox_7
   widget=MotorFrictionName
   properties {
    style=&TextStyle-Base_C
    text=MotorFriction
    font {
     bForceMonospaced=false
     bMaterialIsStencil=false
     fontMaterial=null
     fontObject=&NotoSans
     letterSpacing=0
     monospacedWidth=1
     size=24
     skewAmount=0
     typefaceFontName=NotoSans-Regular
     outlineSettings {
      bApplyOutlineToDropShadows=false
      bMiteredCorners=false
      bSeparateFillAlpha=false
      outlineMaterial=null
      outlineSize=0
      outlineColor=<rgba 0 0 0 1>
     }
    }
    renderTransform {angle=0, scale=<0.7 0.7>, shear=<0 0>, translation=<0 0>}
    shadowOffset=<0 0>
   }
   slot {
    class=VerticalBoxSlot
    properties {horizontalAlignment=HAlign_Left, verticalAlignment=VAlign_Top, padding=<ltrb 10 5 10 0>, size {sizeRule=Fill, value=1}}
   }
  }
  MotorFrictionNum {
   widgetClassPath=&CommonNumericTextBlock
   bIsVariable=true
   parent=VerticalBox_7
   widget=MotorFrictionNum
   properties {
    bIsScrollingEnabled=false
    bIsVolatile=true
    style=&TextStyle-Base_C
    text=Text Block
    font {
     bForceMonospaced=false
     bMaterialIsStencil=false
     fontMaterial=null
     fontObject=&NotoSans
     letterSpacing=0
     monospacedWidth=1
     size=24
     skewAmount=0
     typefaceFontName=NotoSans-Regular
     outlineSettings {
      bApplyOutlineToDropShadows=false
      bMiteredCorners=false
      bSeparateFillAlpha=false
      outlineMaterial=null
      outlineSize=0
      outlineColor=<rgba 0 0 0 1>
     }
    }
    formattingSpecification {
     alwaysSign=false
     maximumFractionalDigits=0
     maximumIntegralDigits=2
     minimumFractionalDigits=2
     minimumIntegralDigits=1
     roundingMode=HalfFromZero
     useGrouping=true
    }
    renderTransform {angle=0, scale=<0.6 0.6>, shear=<0 0>, translation=<0 0>}
    shadowOffset=<0 0>
   }
   slot {
    class=VerticalBoxSlot
    properties {horizontalAlignment=HAlign_Center, verticalAlignment=VAlign_Top, padding=<ltrb 1 1 1 1>, size {sizeRule=Fill, value=1}}
   }
  }
  VerticalBox_9 {widgetClassPath=&VerticalBox, parent=HorizontalBox_93, widget=VerticalBox_9, slot {class=HorizontalBoxSlot}}
  BoostName {
   widgetClassPath=&CommonTextBlock
   bIsVariable=true
   parent=VerticalBox_9
   widget=BoostName
   properties {
    style=&TextStyle-Base_C
    text=Boost
    font {
     bForceMonospaced=false
     bMaterialIsStencil=false
     fontMaterial=null
     fontObject=&NotoSans
     letterSpacing=0
     monospacedWidth=1
     size=24
     skewAmount=0
     typefaceFontName=NotoSans-Regular
     outlineSettings {
      bApplyOutlineToDropShadows=false
      bMiteredCorners=false
      bSeparateFillAlpha=false
      outlineMaterial=null
      outlineSize=0
      outlineColor=<rgba 0 0 0 1>
     }
    }
    renderTransform {angle=0, scale=<0.7 0.7>, shear=<0 0>, translation=<0 0>}
    shadowOffset=<0 0>
   }
   slot {
    class=VerticalBoxSlot
    properties {horizontalAlignment=HAlign_Left, verticalAlignment=VAlign_Top, padding=<ltrb 10 5 10 0>, size {sizeRule=Fill, value=1}}
   }
  }
  BoostNum {
   widgetClassPath=&CommonNumericTextBlock
   bIsVariable=true
   parent=VerticalBox_9
   widget=BoostNum
   properties {
    bIsScrollingEnabled=false
    bIsVolatile=true
    style=&TextStyle-Base_C
    text=Text Block
    font {
     bForceMonospaced=false
     bMaterialIsStencil=false
     fontMaterial=null
     fontObject=&NotoSans
     letterSpacing=0
     monospacedWidth=1
     size=24
     skewAmount=0
     typefaceFontName=NotoSans-Regular
     outlineSettings {
      bApplyOutlineToDropShadows=false
      bMiteredCorners=false
      bSeparateFillAlpha=false
      outlineMaterial=null
      outlineSize=0
      outlineColor=<rgba 0 0 0 1>
     }
    }
    formattingSpecification {
     alwaysSign=false
     maximumFractionalDigits=0
     maximumIntegralDigits=2
     minimumFractionalDigits=2
     minimumIntegralDigits=1
     roundingMode=HalfFromZero
     useGrouping=true
    }
    renderTransform {angle=0, scale=<0.6 0.6>, shear=<0 0>, translation=<0 0>}
    shadowOffset=<0 0>
   }
   slot {
    class=VerticalBoxSlot
    properties {horizontalAlignment=HAlign_Center, verticalAlignment=VAlign_Top, padding=<ltrb 1 1 1 1>, size {sizeRule=Fill, value=1}}
   }
  }
  VerticalBox_10 {widgetClassPath=&VerticalBox, parent=HorizontalBox_93, widget=VerticalBox_10, slot {class=HorizontalBoxSlot}}
  Clutch {
   widgetClassPath=&CommonTextBlock
   bIsVariable=true
   parent=VerticalBox_10
   widget=Clutch
   properties {
    style=&TextStyle-Base_C
    text=Clutch
    font {
     bForceMonospaced=false
     bMaterialIsStencil=false
     fontMaterial=null
     fontObject=&NotoSans
     letterSpacing=0
     monospacedWidth=1
     size=24
     skewAmount=0
     typefaceFontName=NotoSans-Regular
     outlineSettings {
      bApplyOutlineToDropShadows=false
      bMiteredCorners=false
      bSeparateFillAlpha=false
      outlineMaterial=null
      outlineSize=0
      outlineColor=<rgba 0 0 0 1>
     }
    }
    renderTransform {angle=0, scale=<0.7 0.7>, shear=<0 0>, translation=<0 0>}
    shadowOffset=<0 0>
   }
   slot {
    class=VerticalBoxSlot
    properties {horizontalAlignment=HAlign_Left, verticalAlignment=VAlign_Top, padding=<ltrb 10 5 10 0>, size {sizeRule=Fill, value=1}}
   }
  }
  ClutchValue {
   widgetClassPath=&CommonTextBlock
   bIsVariable=true
   parent=VerticalBox_10
   widget=ClutchValue
   properties {
    style=&TextStyle-Base_C
    text=False
    font {
     bForceMonospaced=false
     bMaterialIsStencil=false
     fontMaterial=null
     fontObject=&NotoSans
     letterSpacing=0
     monospacedWidth=1
     size=24
     skewAmount=0
     typefaceFontName=NotoSans-Regular
     outlineSettings {
      bApplyOutlineToDropShadows=false
      bMiteredCorners=false
      bSeparateFillAlpha=false
      outlineMaterial=null
      outlineSize=0
      outlineColor=<rgba 0 0 0 1>
     }
    }
    renderTransform {angle=0, scale=<0.6 0.6>, shear=<0 0>, translation=<0 0>}
    shadowOffset=<0 0>
   }
   slot {
    class=VerticalBoxSlot
    properties {horizontalAlignment=HAlign_Center, verticalAlignment=VAlign_Top, padding=<ltrb 1 1 1 1>, size {sizeRule=Fill, value=1}}
   }
  }
  VerticalBox_11 {widgetClassPath=&VerticalBox, parent=HorizontalBox_93, widget=VerticalBox_11, slot {class=HorizontalBoxSlot}}
  GearName_12 {
   widgetClassPath=&CommonTextBlock
   bIsVariable=true
   parent=VerticalBox_11
   widget=GearName_12
   properties {
    style=&TextStyle-Base_C
    text=CanShift
    font {
     bForceMonospaced=false
     bMaterialIsStencil=false
     fontMaterial=null
     fontObject=&NotoSans
     letterSpacing=0
     monospacedWidth=1
     size=24
     skewAmount=0
     typefaceFontName=NotoSans-Regular
     outlineSettings {
      bApplyOutlineToDropShadows=false
      bMiteredCorners=false
      bSeparateFillAlpha=false
      outlineMaterial=null
      outlineSize=0
      outlineColor=<rgba 0 0 0 1>
     }
    }
    renderTransform {angle=0, scale=<0.7 0.7>, shear=<0 0>, translation=<0 0>}
    shadowOffset=<0 0>
   }
   slot {
    class=VerticalBoxSlot
    properties {horizontalAlignment=HAlign_Left, verticalAlignment=VAlign_Top, padding=<ltrb 10 5 10 0>, size {sizeRule=Fill, value=1}}
   }
  }
  ShiftValue {
   widgetClassPath=&CommonTextBlock
   bIsVariable=true
   parent=VerticalBox_11
   widget=ShiftValue
   properties {
    style=&TextStyle-Base_C
    text=False
    font {
     bForceMonospaced=false
     bMaterialIsStencil=false
     fontMaterial=null
     fontObject=&NotoSans
     letterSpacing=0
     monospacedWidth=1
     size=24
     skewAmount=0
     typefaceFontName=NotoSans-Regular
     outlineSettings {
      bApplyOutlineToDropShadows=false
      bMiteredCorners=false
      bSeparateFillAlpha=false
      outlineMaterial=null
      outlineSize=0
      outlineColor=<rgba 0 0 0 1>
     }
    }
    renderTransform {angle=0, scale=<0.6 0.6>, shear=<0 0>, translation=<0 0>}
    shadowOffset=<0 0>
   }
   slot {
    class=VerticalBoxSlot
    properties {horizontalAlignment=HAlign_Center, verticalAlignment=VAlign_Top, padding=<ltrb 1 1 1 1>, size {sizeRule=Fill, value=1}}
   }
  }
  VerticalBox_12 {widgetClassPath=&VerticalBox, parent=HorizontalBox_93, widget=VerticalBox_12, slot {class=HorizontalBoxSlot}}
  CanDrive {
   widgetClassPath=&CommonTextBlock
   bIsVariable=true
   parent=VerticalBox_12
   widget=CanDrive
   properties {
    style=&TextStyle-Base_C
    text=CanShift
    font {
     bForceMonospaced=false
     bMaterialIsStencil=false
     fontMaterial=null
     fontObject=&NotoSans
     letterSpacing=0
     monospacedWidth=1
     size=24
     skewAmount=0
     typefaceFontName=NotoSans-Regular
     outlineSettings {
      bApplyOutlineToDropShadows=false
      bMiteredCorners=false
      bSeparateFillAlpha=false
      outlineMaterial=null
      outlineSize=0
      outlineColor=<rgba 0 0 0 1>
     }
    }
    renderTransform {angle=0, scale=<0.7 0.7>, shear=<0 0>, translation=<0 0>}
    shadowOffset=<0 0>
   }
   slot {
    class=VerticalBoxSlot
    properties {horizontalAlignment=HAlign_Left, verticalAlignment=VAlign_Top, padding=<ltrb 10 5 10 0>, size {sizeRule=Fill, value=1}}
   }
  }
  DriveValue {
   widgetClassPath=&CommonTextBlock
   bIsVariable=true
   parent=VerticalBox_12
   widget=DriveValue
   properties {
    style=&TextStyle-Base_C
    text=False
    font {
     bForceMonospaced=false
     bMaterialIsStencil=false
     fontMaterial=null
     fontObject=&NotoSans
     letterSpacing=0
     monospacedWidth=1
     size=24
     skewAmount=0
     typefaceFontName=NotoSans-Regular
     outlineSettings {
      bApplyOutlineToDropShadows=false
      bMiteredCorners=false
      bSeparateFillAlpha=false
      outlineMaterial=null
      outlineSize=0
      outlineColor=<rgba 0 0 0 1>
     }
    }
    renderTransform {angle=0, scale=<0.6 0.6>, shear=<0 0>, translation=<0 0>}
    shadowOffset=<0 0>
   }
   slot {
    class=VerticalBoxSlot
    properties {horizontalAlignment=HAlign_Center, verticalAlignment=VAlign_Top, padding=<ltrb 1 1 1 1>, size {sizeRule=Fill, value=1}}
   }
  }
 ]
}
