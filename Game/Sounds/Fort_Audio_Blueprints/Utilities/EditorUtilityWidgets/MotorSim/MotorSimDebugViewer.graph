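# Editor utility widget that live-debugs the sim components of an Audio Motor Model component on the local player's vehicle.
# Each tick it tracks the PIE or editor world, resetting the watched component on change.
# With auto-watch on, it takes the component from the viewed pawn's vehicle or an attached actor.
# It rebuilds the list when the component count changes, then updates each row with its current and previous sim entry.
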EditorUtilityWidgetBlueprint /Game/Sounds/Fort_Audio_Blueprints/Utilities/EditorUtilityWidgets/MotorSim/MotorSimDebugViewer
graphs [
 EventGraph {nodes [
   Event Tick()
   > n1 = Sequence!(execute) -> (then_0, then_1, then_2:Exec, then_3:Exec)
   > Delta Gate Wildcard!(
    execute:Exec
    Variable:"Object Reference (by ref)"=Get CurrentWorld() -> (CurrentWorld:Object Reference)
    New Value:Object Reference=GetCurrentWorld!() -> (ReturnValue:World Object Reference)
   ) -> (Success:Exec, Updated Variable:"Object Reference (by ref)", Previous Variable:Object Reference) {is=title K2Node_MacroInstance}
   > Set WatchedComponent(WatchedComponent:Audio Motor Model Component Object Reference) -> (Output_Get:Audio Motor Model Component Object Reference)
   Branch(execute=@n1.then_2, Condition=Get bAutoWatchLocalPlayer() -> (bAutoWatchLocalPlayer:bool))
   n2 = GetFirstLocalFortPlayerController!(WorldContextObject:Object Reference=Get CurrentWorld(), bIncludeNonSpectators:bool=true, bIncludeSpectators:bool=false) -> (ReturnValue:Fort Player Controller Object Reference)
   > Is Valid(InputObject=@n2)
   n3 = GetViewTarget!(self:Controller Object Reference=@n2) -> (ReturnValue:Actor Object Reference)
   > Is Valid(InputObject=@n3)
   > n4 = Cast To FortPlayerPawn(Object=@n3)
   n5 = Get Vehicle!(self:Fort Player Pawn Object Reference=@n4) -> (ReturnValue:Fort Athena Vehicle Object Reference) {is=title K2Node_CallFunction}
   > Is Valid(InputObject=@n5)
   n6 = GetComponentByClass!(self:Actor Object Reference=@n5, ComponentClass:Actor Component Class Reference=/Script/AudioMotorSim.AudioMotorModelComponent) -> (ReturnValue:Audio Motor Model Component Object Reference)
   > n7 = Is Valid(InputObject=@n6)
   n8 = For Each Loop with Break!(
    Exec:Exec=@n7.Is Not Valid
    Array:Array of Actor Object References=GetAttachedActors!(self:Actor Object Reference=@n5, bResetArray:bool=true, bRecursivelyIncludeAttachedActors:bool=false) -> (OutActors:Array of Actor Object References)
    Break:Exec=@n10.then
   ) -> (LoopBody:Exec, Array Element:Actor Object Reference, Array Index:int, Completed:Exec)
   n9 = GetComponentByClass(self=@n8, ComponentClass=&AudioMotorModelComponent)
   > Is Valid(InputObject=@n9)
   > n10 = Set WatchedComponent(WatchedComponent=@n9)
   Set WatchedComponent(execute=@n7.Is Valid, WatchedComponent=@n6)
   n11 = Get!(execute:Exec=@n1.then_3) -> (then:Exec, else:Exec, WatchedComponent:Audio Motor Model Component Object Reference)
   n12 = Get ComponentNameText() -> (ComponentNameText:Text Block Object Reference)
   > Set Text(Text="To Text (String)"(InString=GetDisplayName(Object=@n11)), self=@n12)
   n13 = Get SimComponents!(self:Audio Motor Model Component Object Reference=@n11) -> (SimComponents:Array of Motor Sim Entry Structures)
   n14 = Get ActorList() -> (ActorList:List View Object Reference)
   > n15 = Branch(Condition="Not Equal (Integer)"(A=Length(TargetArray:Array of Motor Sim Entry Structures=@n13), B=GetNumItems!(self:List View Object Reference=@n14) -> (ReturnValue:int=0)))
   > ClearListItems(self=@n14)
   > n16 = For Each Loop(Array:Array of Motor Sim Entry Structures=@n13) -> (Array Element:Motor Sim Entry Structure)
   n17 = For Each Loop(Exec=[@n15.else, @n16.Completed], Array=@n13)
   > n18 = Cast To MotorSimDebugComponentEntry!(
    execute
    Object="Get (a copy)"(
     Array:Array of User Widget Object References=GetDisplayedEntryWidgets!(self:List View Base Object Reference=@n14) -> (ReturnValue:Array of User Widget Object References)
     Dimension 1=@n17.Array Index
    ) -> (Output:User Widget Object Reference)
   ) -> (then, CastFailed, AsMotor Sim Debug Component Entry:Motor Sim Debug Component Entry Object Reference) {is=title K2Node_DynamicCast}
   n19 = int - int(A=@n17.Array Index)
   > UpdateText!(
    execute
    self:Motor Sim Debug Component Entry Object Reference=@n18
    SimEntry:Motor Sim Entry Structure=@n17
    PrevSimEntry:Motor Sim Entry Structure="Get (a ref)"(Array:Array of Motor Sim Entry Structures=@n13, Dimension 1="Clamp (Integer)"(Value=@n19, Min=0, Max=99) -> (ReturnValue=0)) -> (Output:"Motor Sim Entry Structure (by ref)")
   ) -> (then)
   n20 = Construct Object from Class(
    execute=@n16.LoopBody
    Class=&/MotorSimDebugComponentEntry.MotorSimDebugComponentEntry_C
   ) -> (ReturnValue:Motor Sim Debug Component Entry Object Reference)
   > AddItem!(execute, self:List View Object Reference=@n14, Item:Object Reference=@n20) -> (then)
   Set Text(execute=@n11.else, Text="LOCTEXT(\"No Vehicle : (\")", self=@n12)
   Event PreConstruct()
   Event Construct()
  ]}
 GetCurrentWorld {nodes [
   GetCurrentWorld() {is=title K2Node_FunctionEntry}
   > n1 = GetPIEWorlds!(execute, bIncludeDedicatedServer:bool=false) -> (then, ReturnValue:Array of World Object References)
   > n2 = Branch(Condition=Is Valid Index!(TargetArray:Array of World Object References=@n1, IndexToTest:int) -> (ReturnValue:bool))
   > Return Node!(execute, ReturnValue:World Object Reference="Get (a copy)"(Array:Array of World Object References=@n1) -> (Output:World Object Reference))
   n3 = GetEditorWorld(execute=@n2.else, self=Get UnrealEditorSubsystem())
   > Return Node(ReturnValue=@n3)
  ]}
]
variables [
 bAutoWatchLocalPlayer {type=boolean, default=true}
 CurrentWorld {type=Object, default=null}
 PrevComponentListLength {type=integer}
 WatchedComponent {type=AudioMotorModelComponent, default=null}
]
widgets {
 info {rootWidgetClass=&CanvasPanel, widgetCount=4}
 widgets [
  CanvasPanel_0 {widgetClassPath=&CanvasPanel, widget=CanvasPanel_0}
  VerticalBox_0 {
   widgetClassPath=&VerticalBox
   parent=CanvasPanel_0
   widget=VerticalBox_0
   slot {
    class=CanvasPanelSlot
    properties {layoutData {alignment=<0 0>, anchors {maximum=<0 0>, minimum=<0 0>}, offsets=<ltrb 8 8 1906.3063 373.3333>}}
   }
  }
  ComponentNameText {
   widgetClassPath=&UEFN_TextBlock
   bIsVariable=true
   parent=VerticalBox_0
   widget=ComponentNameText
   properties {text=ModelName}
   slot {class=VerticalBoxSlot}
  }
  ActorList {
   widgetClassPath=/Script/UMG.ListView
   bIsVariable=true
   parent=VerticalBox_0
   widget=ActorList
   properties {
    bAllowDragging=false
    bIsFocusable=false
    bIsVolatile=true
    clipping=Inherit
    entryWidgetClass=&MotorSimDebugComponentEntry_C
    selectionMode=Multi
    scrollBarStyle {
     thickness=8
     draggedThumbImage {
      bIsDynamicallyLoaded=false
      drawAs=RoundedBox
      imageType=NoImage
      mirroring=NoMirror
      resourceName=None
      resourceObject=null
      tiling=NoTile
      imageSize=<0 0>
      margin=<ltrb 0 0 0 0>
      outlineSettings {
       bUseBrushTransparency=false
       roundingType=FixedRadius
       width=0
       color {colorUseRule=UseColor_Specified, specifiedColor=<rgba 0 0 0 0>}
       cornerRadii=<4 4 4 4>
      }
      tintColor {colorUseRule=UseColor_Specified, specifiedColor=<rgba 0.2159 0.2159 0.2159 1>}
      uVRegion {bIsValid=false, max=<0 0>, min=<0 0>}
     }
     horizontalBackgroundImage {
      bIsDynamicallyLoaded=false
      drawAs=NoDrawType
      imageType=NoImage
      mirroring=NoMirror
      resourceName=None
      resourceObject=null
      tiling=NoTile
      imageSize=<0 0>
      margin=<ltrb 0 0 0 0>
      outlineSettings {
       bUseBrushTransparency=false
       roundingType=HalfHeightRadius
       width=0
       color {colorUseRule=UseColor_Specified, specifiedColor=<rgba 0 0 0 0>}
       cornerRadii=<0 0 0 0>
      }
      tintColor {colorUseRule=UseColor_Specified, specifiedColor=<rgba 1 1 1 1>}
      uVRegion {bIsValid=false, max=<0 0>, min=<0 0>}
     }
     horizontalBottomSlotImage {
      bIsDynamicallyLoaded=false
      drawAs=NoDrawType
      imageType=NoImage
      mirroring=NoMirror
      resourceName=None
      resourceObject=null
      tiling=NoTile
      imageSize=<0 0>
      margin=<ltrb 0 0 0 0>
      outlineSettings {
       bUseBrushTransparency=false
       roundingType=HalfHeightRadius
       width=0
       color {colorUseRule=UseColor_Specified, specifiedColor=<rgba 0 0 0 0>}
       cornerRadii=<0 0 0 0>
      }
      tintColor {colorUseRule=UseColor_Specified, specifiedColor=<rgba 1 1 1 1>}
      uVRegion {bIsValid=false, max=<0 0>, min=<0 0>}
     }
     horizontalTopSlotImage {
      bIsDynamicallyLoaded=false
      drawAs=NoDrawType
      imageType=NoImage
      mirroring=NoMirror
      resourceName=None
      resourceObject=null
      tiling=NoTile
      imageSize=<0 0>
      margin=<ltrb 0 0 0 0>
      outlineSettings {
       bUseBrushTransparency=false
       roundingType=HalfHeightRadius
       width=0
       color {colorUseRule=UseColor_Specified, specifiedColor=<rgba 0 0 0 0>}
       cornerRadii=<0 0 0 0>
      }
      tintColor {colorUseRule=UseColor_Specified, specifiedColor=<rgba 1 1 1 1>}
      uVRegion {bIsValid=false, max=<0 0>, min=<0 0>}
     }
     hoveredThumbImage {
      bIsDynamicallyLoaded=false
      drawAs=RoundedBox
      imageType=NoImage
      mirroring=NoMirror
      resourceName=None
      resourceObject=null
      tiling=NoTile
      imageSize=<0 0>
      margin=<ltrb 0 0 0 0>
      outlineSettings {
       bUseBrushTransparency=false
       roundingType=FixedRadius
       width=0
       color {colorUseRule=UseColor_Specified, specifiedColor=<rgba 0 0 0 0>}
       cornerRadii=<4 4 4 4>
      }
      tintColor {colorUseRule=UseColor_Specified, specifiedColor=<rgba 0.2159 0.2159 0.2159 1>}
      uVRegion {bIsValid=false, max=<0 0>, min=<0 0>}
     }
     normalThumbImage {
      bIsDynamicallyLoaded=false
      drawAs=RoundedBox
      imageType=NoImage
      mirroring=NoMirror
      resourceName=None
      resourceObject=null
      tiling=NoTile
      imageSize=<0 0>
      margin=<ltrb 0 0 0 0>
      outlineSettings {
       bUseBrushTransparency=false
       roundingType=FixedRadius
       width=0
       color {colorUseRule=UseColor_Specified, specifiedColor=<rgba 0 0 0 0>}
       cornerRadii=<4 4 4 4>
      }
      tintColor {colorUseRule=UseColor_Specified, specifiedColor=<rgba 0.0953 0.0953 0.0953 1>}
      uVRegion {bIsValid=false, max=<0 0>, min=<0 0>}
     }
     verticalBackgroundImage {
      bIsDynamicallyLoaded=false
      drawAs=NoDrawType
      imageType=NoImage
      mirroring=NoMirror
      resourceName=None
      resourceObject=null
      tiling=NoTile
      imageSize=<0 0>
      margin=<ltrb 0 0 0 0>
      outlineSettings {
       bUseBrushTransparency=false
       roundingType=HalfHeightRadius
       width=0
       color {colorUseRule=UseColor_Specified, specifiedColor=<rgba 0 0 0 0>}
       cornerRadii=<0 0 0 0>
      }
      tintColor {colorUseRule=UseColor_Specified, specifiedColor=<rgba 1 1 1 1>}
      uVRegion {bIsValid=false, max=<0 0>, min=<0 0>}
     }
     verticalBottomSlotImage {
      bIsDynamicallyLoaded=false
      drawAs=NoDrawType
      imageType=NoImage
      mirroring=NoMirror
      resourceName=None
      resourceObject=null
      tiling=NoTile
      imageSize=<0 0>
      margin=<ltrb 0 0 0 0>
      outlineSettings {
       bUseBrushTransparency=false
       roundingType=HalfHeightRadius
       width=0
       color {colorUseRule=UseColor_Specified, specifiedColor=<rgba 0 0 0 0>}
       cornerRadii=<0 0 0 0>
      }
      tintColor {colorUseRule=UseColor_Specified, specifiedColor=<rgba 1 1 1 1>}
      uVRegion {bIsValid=false, max=<0 0>, min=<0 0>}
     }
     verticalTopSlotImage {
      bIsDynamicallyLoaded=false
      drawAs=NoDrawType
      imageType=NoImage
      mirroring=NoMirror
      resourceName=None
      resourceObject=null
      tiling=NoTile
      imageSize=<0 0>
      margin=<ltrb 0 0 0 0>
      outlineSettings {
       bUseBrushTransparency=false
       roundingType=HalfHeightRadius
       width=0
       color {colorUseRule=UseColor_Specified, specifiedColor=<rgba 0 0 0 0>}
       cornerRadii=<0 0 0 0>
      }
      tintColor {colorUseRule=UseColor_Specified, specifiedColor=<rgba 1 1 1 1>}
      uVRegion {bIsValid=false, max=<0 0>, min=<0 0>}
     }
    }
   }
   slot {class=VerticalBoxSlot}
  }
 ]
}
